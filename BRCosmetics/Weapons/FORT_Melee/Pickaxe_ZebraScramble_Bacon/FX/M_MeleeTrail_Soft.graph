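# Translucent, unlit, two-sided material for Niagara ribbons, a soft melee swing trail for a pickaxe cosmetic.
# Particle color supplies RGB through an emissive adjustment function and scales alpha.
# Two noise samples panning at different speeds multiply together, then modulate a procedural UV mask that shapes both ends of U and both edges of V.
# An alpha multiplier and depth fade soften the resulting opacity.
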Material /BRCosmetics/Weapons/FORT_Melee/Pickaxe_ZebraScramble_Bacon/FX/M_MeleeTrail_Soft
graph {
 properties {
  bUsedWithNiagaraRibbons=true
  blendMode=BLEND_Translucent
  opacityMaskClipValue=0.1
  pixelDepthOffsetMode=PDOM_Legacy
  shadingModel=MSM_Unlit
  twoSided=true
  usedShadingModels=Unlit
 }
 nodes [
  n1 = ParticleColor()
  n2 = TextureCoordinate() {uTiling=0.2, vTiling=0.5}
  n3 = TextureCoordinate() {uTiling=0.3}
  n4 = Multiply(
   A=TextureSampleParameter2D(UVs=Panner(Coordinate=@n2) {speedX=-0.6, speedY=-0.05}) {parameterName=NoiseTexture, samplerType=SAMPLERTYPE_LinearColor, texture=&HighContrastNoise}.R
   B=TextureSampleParameter2D(UVs=Panner(Coordinate=@n3) {speedX=-0.3, speedY=0.05}) {parameterName=NoiseTexture, samplerType=SAMPLERTYPE_LinearColor, texture=&HighContrastNoise}.B
  )
  n5 = TextureCoordinate()
  n6 = ComponentMask(None=@n5) {r=true}
  n7 = OneMinus(None=@n6)
  n8 = ComponentMask(None=@n5) {g=true}
  n9 = OneMinus(None=@n8)
  n10 = Multiply(A=Multiply(A=Divide(A=@n6) {constB=0.1}, B=Power(Base=@n7) {constExponent=3}), B=Divide(A=Multiply(A=@n8, B=@n9)) {constB=0.25})
  n11 = Multiply(A=Multiply(A=Add(A=@n4) {constB=0.5}, B=@n10), B=ScalarParameter() {defaultValue=0.6, parameterName=AlphaMultiplier}) {constB=0.65}
  DepthFade(Opacity=Multiply(A=@n1.A, B=Saturate(None=@n11)), FadeDistance=ScalarParameter() {parameterName=DepthFadeDistance, sliderMax=100})
  MaterialFunctionCall(Emissive=@n1.RGB) {materialFunction=&MF_FX_EmissiveAdjustments}
  Multiply(A=@n10) {constB=2}
  Multiply(A=@n7, B=@n7)
 ]
}
